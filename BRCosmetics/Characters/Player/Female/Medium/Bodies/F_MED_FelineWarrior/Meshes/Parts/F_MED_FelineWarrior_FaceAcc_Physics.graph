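# Physics asset for the female medium FelineWarrior face accessory mesh, defining simulation bodies for its dynamic hair bone chain.
# Six mostly tapering capsules on dyn_hair_0 to dyn_hair_5 form a chain anchored to neck_01, each link constrained by identical tight swing and twist limits.
# Larger capsules on neck_01, spine_04 and spine_05 (which has two) approximate the neck and upper torso; the spine bodies have no constraints.
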PhysicsAsset /BRCosmetics/Characters/Player/Female/Medium/Bodies/F_MED_FelineWarrior/Meshes/Parts/F_MED_FelineWarrior_FaceAcc_Physics
bodies {
 dyn_hair_0 [{
   shapeType=Capsule
   length=2.2622
   radius=2.2378
   shapeName=dyn_hair_0_capsule
   center=<1.9716 -0.3189 -0.6347>
   rotation=<rot 90 180 174.195>
  }]
 dyn_hair_1 [{
   shapeType=Capsule
   length=2.1537
   radius=2.3463
   shapeName=dyn_hair_1_capsule
   center=<1.8554 0.7086 -0.7988>
   rotation=<rot 76.9942 0.0002 0.0002>
  }]
 dyn_hair_2 [{
   shapeType=Capsule
   length=2.4
   radius=2.1
   shapeName=dyn_hair_2_capsule
   center=<2.0078 0.3322 -0.1849>
   rotation=<rot 89.3726 6.3127 7.0808>
  }]
 dyn_hair_3 [{shapeType=Capsule, length=3.2, radius=1.3, shapeName=dyn_hair_3_capsule, center=<1.9143 0 -0.5188>, rotation=<rot 80.614 0 0>}]
 dyn_hair_4 [{shapeType=Capsule, length=3.2, radius=1.3, shapeName=dyn_hair_4_capsule, center=<2 0 0>, rotation=<rot 88.1121 -179.9994 -179.9994>}]
 dyn_hair_5 [{shapeType=Capsule, length=3.6, radius=0.9, shapeName=dyn_hair_5_capsule, center=<2 0 0>, rotation=<rot 90 0 0>}]
 neck_01 [{shapeType=Capsule, length=8.7911, radius=5.2089, shapeName=neck_01_capsule, center=<0 1.3701 2.4975>, rotation=<rot 90 0 0>}]
 spine_04 [{shapeType=Capsule, length=19.3, radius=10.7, shapeName=spine_04_capsule, center=<2.0789 2.2165 0>}]
 spine_05 [
  {shapeType=Capsule, length=19.3, radius=10.7, shapeName=spine_05_capsule, center=<0.1654 -0.3835 0>}
  {
   shapeType=Capsule
   length=17.5537
   radius=5.4735
   shapeName=spine_05_capsule
   center=<7.658 4.8807 -10.6126>
   rotation=<rot 63.3154 25.1951 71.6273>
  }
 ]
}
constraints [
 {bone1Name=dyn_hair_0, bone2Name=neck_01, swing1LimitDegrees=5, swing2LimitDegrees=10, twistLimitDegrees=2}
 {bone1Name=dyn_hair_1, bone2Name=dyn_hair_0, swing1LimitDegrees=5, swing2LimitDegrees=10, twistLimitDegrees=2}
 {bone1Name=dyn_hair_2, bone2Name=dyn_hair_1, swing1LimitDegrees=5, swing2LimitDegrees=10, twistLimitDegrees=2}
 {bone1Name=dyn_hair_3, bone2Name=dyn_hair_2, swing1LimitDegrees=5, swing2LimitDegrees=10, twistLimitDegrees=2}
 {bone1Name=dyn_hair_4, bone2Name=dyn_hair_3, swing1LimitDegrees=5, swing2LimitDegrees=10, twistLimitDegrees=2}
 {bone1Name=dyn_hair_5, bone2Name=dyn_hair_4, swing1LimitDegrees=5, swing2LimitDegrees=10, twistLimitDegrees=2}
]
