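# Animation Blueprint for the Feline Warrior face accessory mesh, adding secondary physics and corrective posing to its dynamic hair bones.
# It copies the attached parent's pose, then runs RigidBody simulation in neck_01 space with damped component motion.
# Each update converts head twist relative to spine_05 into an alpha.
# That alpha drives chained additive bone-space rotations on dyn_hair_0, dyn_hair_1 and dyn_hair_2 after the simulation.
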AnimBlueprint /BRCosmetics/Characters/Player/Female/Medium/Bodies/F_MED_FelineWarrior/Meshes/Parts/F_MED_FelineWarrior_FaceAcc_AnimBP
target_skeleton=&/F_MED_FelineWarrior_FaceAcc_Skeleton
graphs [
 AnimGraph {nodes [
   n1 = RigidBody(ComponentPose=Local To Component(LocalPose=Copy Pose From Mesh() {bUseAttachedParent=true})) {
    bClampLinearTranslationLimitToRefPose=true
    bOverrideWorldGravity=true
    overridePhysicsAsset=&/F_MED_FelineWarrior_FaceAcc_Physics
    simulationSpace=BaseBoneSpace
    alphaBoolBlend {blendOption=Linear}
    baseBoneRef {boneName=neck_01}
    componentAppliedLinearAccClamp=<3000 3000 3000>
    componentLinearAccScale=<0.1 0.1 0.1>
    componentLinearVelScale=<0.1 0.1 0.1>
    overrideWorldGravity {z=-980}
   }
   n2 = Get Alpha Output() -> (Alpha Output:float)
   n3 = "Transform (Modify) Bone - Bone: dyn_hair_0"(ComponentPose=@n1, Alpha=@n2, Translation="0,0,0", Rotation="-10.857,-32.360,-0.095", Scale="1,1,1") {
    rotationMode=BMM_Additive
    rotationSpace=BCS_BoneSpace
    alphaBoolBlend {blendOption=Linear}
    boneToModify {boneName=dyn_hair_0}
    rotation=<rot -10.857 -32.36 -0.095>
   }
   n4 = "Transform (Modify) Bone - Bone: dyn_hair_1"(ComponentPose=@n3, Alpha=@n2, Translation="0,0,0", Rotation="-0.028,17.278,-0.185", Scale="1,1,1") {
    rotationMode=BMM_Additive
    rotationSpace=BCS_BoneSpace
    alphaBoolBlend {blendOption=Linear}
    boneToModify {boneName=dyn_hair_1}
    rotation=<rot -0.0279 17.278 -0.1849>
   }
   n5 = "Transform (Modify) Bone - Bone: dyn_hair_2"(ComponentPose=@n4, Alpha=@n2, Translation="0,0,0", Rotation="13.488,34.597,4.061", Scale="1,1,1") {
    rotationMode=BMM_Additive
    rotationSpace=BCS_BoneSpace
    alphaBoolBlend {blendOption=Linear}
    boneToModify {boneName=dyn_hair_2}
    rotation=<rot 13.488 34.597 4.061>
   }
   Output Pose(Result=Component To Local(ComponentPose=@n5)) {layerGroup=None}
  ]}
 EventGraph {nodes [
   Event BlueprintUpdateAnimation()
   > Set Alpha Output(Alpha Output:float=GetAlphaFromReferenceTransformDelta(
     SkeletalMeshComponent=GetOwnerSkeletalMeshComponent()
     BoneName=head
     BaseName=spine_05
     DeltaComponentType=QuaternionTwist
     InputRange=(X=0,Y=50)
     TwistAxis="1,0,0"
    )) -> (Output_Get:float)
  ]}
]
variables [Alpha Output {type=number}]
